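; Athena frag grenade projectile that explodes when its fuse ends, extending the base grenade with audio and effects.
; At half fuse time, a cosmetic-only fuse-audio ramp runs by timeline, repeating timer or instant set, chosen by an optimization enum.
; Fuse end scales the radius effect, tinted red for non-teammates, then explodes with surface-based sound, feedback cue and optional physics field.
; Bounces spawn surface-dependent impact FX; water entry activates floating Niagara FX.
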(event Custom|Pre%20Explo%20Audio%20Tell
  ;[1 1 1] Ramp up the audio right before explo
  (Networking|CanExecuteCosmeticEvents
    (:True
      (Variables|Default|SetPreExplodeRampTime (/ 1.0 (/ (Variables|Default|GetFuseTime) 2.0)))
      (switch Utilities|FlowControl|Switch|SwitchonPreExplodeRampType (Variables|Default|GetPreExplodeOptimization)
        (:NewEnumerator0
          ;;[1 1 1] Comment
          (Components|Timeline|SetPlayRate (Variables|B_Prj_Athena_FragGrenade|GetPreExploWarning) (Variables|Default|GetPreExplodeRampTime))
          (bind (_direction _preexplo) (Timeline|PreExploWarning PlayFromStart)
            (:Update
              (Audio|Parameter|SetFloatParameter (Variables|B_Prj_Grenade|GetGrenadeFuse_AudioComponent) (Variables|Default|GetGrenadeFuseParamName) _preexplo :!class /Script/AudioExtensions.AudioParameterControllerInterface))))
        (:NewEnumerator1
          ;;[1 1 1] Comment
          (Utilities|Time|SetTimerbyEvent (EventDispatchers|CreateEvent :!event Custom|TimerPreExplode) (Variables|Default|GetPreExplodeTimerRampFrequency) false true))
        (:NewEnumerator3
          ;;[1 1 1] Comment
          (Audio|Parameter|SetFloatParameter (Variables|B_Prj_Grenade|GetGrenadeFuse_AudioComponent) (Variables|Default|GetGrenadeFuseParamName) 1.0 :!class /Script/AudioExtensions.AudioParameterControllerInterface))))))

(event Custom|TimerPreExplode
  ;[1 1 1] Ramp up the audio right before explo
  ;;[1 1 1] Comment
  (Audio|Parameter|SetFloatParameter (Variables|B_Prj_Grenade|GetGrenadeFuse_AudioComponent) (Variables|Default|GetGrenadeFuseParamName) (Variables|Default|GetPreExplodeTimerAccumulatedRampTime) :!class /Script/AudioExtensions.AudioParameterControllerInterface)
  (bind _output_get (Variables|Default|SetPreExplodeTimerAccumulatedRampTime (+ (Variables|Default|GetPreExplodeTimerAccumulatedRampTime) (/ (Variables|Default|GetPreExplodeTimerRampFrequency) (Variables|Default|GetPreExplodeRampTime)))))
  (Utilities|FlowControl|Sequence
    (:then_0
      (Audio|Parameter|SetFloatParameter (Variables|B_Prj_Grenade|GetGrenadeFuse_AudioComponent) (Variables|Default|GetGrenadeFuseParamName) _output_get :!class /Script/AudioExtensions.AudioParameterControllerInterface))
    (:then_1
      (if (< _output_get 1.0)
        (Utilities|FlowControl|Sequence
          (:then_0)
          (:then_1
            (Utilities|Time|SetTimerbyEvent (EventDispatchers|CreateEvent :!event Custom|TimerPreExplode) (Variables|Default|GetPreExplodeTimerRampFrequency) false true)))
        (else
          _)))))

(event EventBeginPlay
  (|CallParentFunction)
  (Utilities|Time|SetTimerbyEvent (EventDispatchers|CreateEvent :!event Custom|Pre%20Explo%20Audio%20Tell) (/ (Variables|Default|GetFuseTime) 2.0)))

(event EventFuseEnded
  (bind _effect_distance (Variables|B_Prj_Grenade|GetEffect_Distance))
  (Transformation|SetWorldRotation _effect_distance "90.000000,0.000000,0.000000" false false)
  (Components|Activation|Activate _effect_distance false)
  (Effects|Components|ParticleSystem|SetVectorParameter _effect_distance "Size" (Math|Conversions|ToVector(Float) (Variables|Default|GetExplosionRadius)))
  (if (not (Team|OnSameTeam (Game|GetInstigatorController) (Game|GetPlayerController 0)))
    (Effects|Components|ParticleSystem|SetVectorParameter _effect_distance "Color" "1.500000,0.110000,0.190000"))
  (Delivery|DoExplosion)
  (Audio|Components|Audio|FadeOut (Variables|B_Prj_Grenade|GetGrenadeFuse_AudioComponent) 0.3 0.0 "Linear" :!class /Script/Engine.AudioComponent)
  (Components|DestroyComponent (Variables|Default|GetNS_Object_FloatingInWater)))

(event EventOnExploded
  (params (HitActors :type Array%20of%20Actor%20Object%20References) (HitResults :type Array%20of%20Hit%20Result%20Structures))
  (|CallParentFunction HitActors HitResults)
  (Utilities|AllowJanus
    (:Allow
      (bind _returnvalue (Development|GetConsoleVariableIntValue "r.PhysicsField.EnableField"))
      (if (== _returnvalue 1)
        (Game|SpawnActorfromClass "/Game/Effects/Niagara/Blueprints/PhsyicsFields/FN_RadialForce.FN_RadialForce_C" (Transformation|GetWorldTransform (Variables|Collision|GetCapsuleComponent)) "Undefined" "MultiplyWithRoot" 0 0.2 0.66 1.5e+03 2.0 0.0 0)
        (else
          _)))
    (:Nope
      _))
  (Class|BPSurfaceTypeSoundComponent|PlaySound (Variables|Default|GetBP_SurfaceTypeSoundComponent) (Transformation|GetActorLocation))
  (GameplayCue|ExecuteGameplayCue%28Burst%29Local self (Variables|Default|GetFeedbackCue) (Ability|GameplayCue|MakeGameplayCueParameters 0.0)) ; Cam shake + controller vibration
)

(event EventOnBounce
  (params (Hit :type Hit%20Result%20Structure%20%28by%20ref%29))
  (bind (_bblockinghit _binitialoverlap _time _distance _location _impactpoint _normal _impactnormal _physmat _hitactor) (Collision|BreakHitResult Hit))
  (bind _surfacetype (Variables|/Script/PhysicsCore.PhysicalMaterial|GetSurfaceType _physmat))
  (|CallParentFunction Hit)
  (Utilities|IsValid _physmat
    (:Is%20Valid
      (Class|BPSurfaceTypeSoundComponent|SetSurfaceType (Variables|Default|GetBP_SurfaceTypeSoundComponent) _surfacetype _hitactor)
      ;[1 0.5052 0.5052] To Do. It's mud, but would better to have other surface type bounce FX
      (if (== _surfacetype 30)
        (Niagara|SpawnSystematLocation (Variables|Default|GetNiagara_Bounse) _location)
        (else
          (Niagara|SpawnSystematLocation "/Game/Effects/Fort_Effects/Effects/Weapon/NS_impact_smoke_Small.NS_impact_smoke_Small" _location))))))

(event OnEnterWaterEvent(WaterInteractionComponent)
  (params (WaterBody :type Fort%20Water%20Body%20Actor%20Object%20Reference) (WaterInteractionComponent :type Fort%20Water%20Interaction%20Component%20Object%20Reference) (bIsFirstBody :type Boolean))
  (bind _ns_object_floatinginwater (Variables|Default|GetNS_Object_FloatingInWater))
  (bind _outwatersurfaceinfo (WaterBody|GetWaterSurfaceInfoStruct WaterBody (Transformation|GetActorLocation) true false false))
  ;[1 1 1] Get water info
  (Components|Activation|Activate _ns_object_floatinginwater false)
  (Effects|Components|ParticleSystem|SetFloatParameter _ns_object_floatinginwater "User.WaterDepth" (.WaterDepth (Utilities|Struct|BreakWaterSurfaceInfo _outwatersurfaceinfo)))
  (Effects|Components|ParticleSystem|SetVectorParameter _ns_object_floatinginwater "User.WaterPlaneLocation" (Utilities|Struct|BreakWaterSurfaceInfo _outwatersurfaceinfo))
  (Effects|Components|ParticleSystem|SetVectorParameter _ns_object_floatinginwater "User.WaterPlaneNormal" (.WaterPlaneNormal (Utilities|Struct|BreakWaterSurfaceInfo _outwatersurfaceinfo)))
  (Effects|Components|ParticleSystem|SetFloatParameter _ns_object_floatinginwater "User.InheritVelocityScale" 0.3))

(event OnExitWaterEvent(WaterInteractionComponent)
  (params (WaterBody :type Fort%20Water%20Body%20Actor%20Object%20Reference) (WaterInteractionComponent :type Fort%20Water%20Interaction%20Component%20Object%20Reference) (bIsLastBody :type Boolean))
  (Components|Activation|Deactivate (Variables|Default|GetNS_Object_FloatingInWater)))

(fn ConstructionScript
  (|CallParentFunction))
