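; Base projectile for Athena thrown grenades, handling fuse, bouncing, water splashes and explosion feedback.
; BeginPlay starts the fuse timer into an empty FuseEnded function; any damage triggers a delayed explosion.
; Bounces are counted, pawn overlaps fake a bounce for AI creatures, and "Bounce"-tagged components boost velocity.
; On explosion, cosmetic-gated logic traces for water to pick VFX, adds surface-specific FX, and plays audio, sound indicator, camera shake and force feedback.
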(event Custom|AysncLoad
  ;[1 1 1] Async Load All VFX
  (Utilities|AsyncLoadAllVFX))

(event EventBeginPlay
  ;[1 1 1] BeginPlay - Start Fuse
  (CallFunction|AysncLoad)
  (Utilities|Time|SetTimerbyEvent (EventDispatchers|CreateEvent :!event FuseEnded) (Variables|Default|GetFuseTime))
  (Water|BindEventtoOnEnterWaterEvent (Variables|Water|GetWaterInteractionComponent) (EventDispatchers|CreateEvent :!event Custom|Splash))
  (Networking|CanExecuteCosmeticEvents
    (:True
      (Audio|Components|Audio|SetSound (Variables|B_Prj_Grenade|GetGrenadeFuse_AudioComponent) (Variables|Sounds|GetCue_GrenadeFuseSound))
      (if (Variables|Default|GetbShowFuseSoundIndicator)
        (bind _returnvalue (Utilities|Time|SetTimerbyEvent (delegate Custom|ShowFuseIndicator) 0.3 true))
        (Variables|Default|SetFuseIndicatorTimer _returnvalue)))))

(event Custom|ShowFuseIndicator
  ;[1 1 1] BeginPlay - Start Fuse
  (Fortnite|ShowSoundIndicatorLocal :TrackedActor self :Location (Transformation|GetActorLocation) :MaxAudibleDistance (Variables|/Script/Engine.SoundBase|GetMaxDistance (Variables|Sounds|GetCue_GrenadeFuseSound)) :IndicatorType "Glider" :Instigator (Variables|Actor|GetInstigator) :Tint "(R=1.000000,G=1.000000,B=1.000000,A=1.000000)" :OverrideIcon (Variables|Default|GetSoundIndicatorIcon) :SoundBase (Variables|Sounds|GetCue_GrenadeFuseSound) :AttenuationMaxRangeMultiplier 1.0 :AttenuationIntensityMultiplier 1.0) ; Indicator for fuse
)

(event OnComponentBeginOverlap(CapsuleComponent)
  (params (OverlappedComponent :type Primitive%20Component%20Object%20Reference) (OtherActor :type Actor%20Object%20Reference) (OtherComp :type Primitive%20Component%20Object%20Reference) (OtherBodyIndex :type Integer) (bFromSweep :type Boolean) (SweepResult :type Hit%20Result%20Structure%20%28by%20ref%29))
  (bind (_bblockinghit _binitialoverlap _time _distance _location _impactpoint _normal _impactnormal _physmat _hitactor) (Collision|BreakHitResult SweepResult))
  (bind _projectilemovementcomponent (Variables|Movement|GetProjectileMovementComponent))
  ;[1 0 0] FortAIPawn type creatures and Deimos use Overlaps. Players will not.            If we overlap a pawn, fake a bounce.
  (bind _aspawn (Utilities|Casting|CastToPawn _hitactor))
  (if (Utilities|NotEqual(Object) _aspawn (Game|GetInstigator))
    (Variables|/Script/Engine.ProjectileMovementComponent|SetProjectileGravityScale (Variables|Default|GetBouncePawnAgainstPawnGravityScale) _projectilemovementcomponent) ; Hit a pawn, scale gravity
    (Variables|/Script/Engine.MovementComponent|SetVelocity (* (Math|Vector|MakeVector (.x (* (Variables|/Script/Engine.MovementComponent|GetVelocity _projectilemovementcomponent) -1)) (.y (* (Variables|/Script/Engine.MovementComponent|GetVelocity _projectilemovementcomponent) -1)) (Math|Float|Absolute(Float) (.z (* (Variables|/Script/Engine.MovementComponent|GetVelocity _projectilemovementcomponent) -1)))) (Variables|/Script/Engine.ProjectileMovementComponent|GetBounciness _projectilemovementcomponent)) (Variables|Movement|GetProjectileMovementComponent))
    (Variables|Default|SetCurrentNumberOfBounces (+ (Variables|Default|GetCurrentNumberOfBounces) 1))))

(event EventOnBounce
  (params (Hit :type Hit%20Result%20Structure%20%28by%20ref%29))
  (bind (_bblockinghit _binitialoverlap _time _distance _location _impactpoint _normal _impactnormal _physmat _hitactor _hitcomponent) (Collision|BreakHitResult Hit))
  ;[1 1 1] OnBounce - After X bounces set to explode
  (Variables|Default|SetE_Surface_Type (Variables|/Script/PhysicsCore.PhysicalMaterial|GetSurfaceType _physmat))
  (bind _aspawn (Utilities|Casting|CastToPawn _hitactor)
    (:then
      (if (Utilities|NotEqual(Object) _aspawn (Game|GetInstigator))
        (Variables|/Script/Engine.ProjectileMovementComponent|SetProjectileGravityScale (Variables|Default|GetBouncePawnAgainstPawnGravityScale) (Variables|Movement|GetProjectileMovementComponent)) ; Hit a pawn, scale gravity
      ))
    (:CastFailed
      (if (Components|ComponentHasTag _hitcomponent "Bounce")
        (Variables|/Script/Engine.MovementComponent|SetVelocity (Math|Vector|vector+vector (* (Variables|/Script/Engine.MovementComponent|GetVelocity (Variables|Movement|GetProjectileMovementComponent)) 2.75) (Math|Vector|MakeVector 0.0 0.0 (* (Math|Vector|VectorLength (Variables|/Script/Engine.MovementComponent|GetVelocity (Variables|Movement|GetProjectileMovementComponent))) 0.8))) (Variables|Movement|GetProjectileMovementComponent))
        (else
          _))))
  (Utilities|FlowControl|Sequence
    (:then_0
      (Variables|Default|SetCurrentNumberOfBounces (+ (Variables|Default|GetCurrentNumberOfBounces) 1))
      (Utilities|IsValid (Variables|Sounds|GetCue_Bounce)
        (:Is%20Valid
          (Audio|OnCollision (Variables|Default|GetFortCollisionAudio) _location (* _impactnormal (Variables|/Script/Engine.MovementComponent|GetVelocity (Variables|Movement|GetProjectileMovementComponent)))))))))

(event Custom|Stop_Rotation
  ;[1 1 1] Stop Rotating Movement
  (Components|Movement|SetUpdatedComponent (Variables|Default|GetRotatingMovement) 0))

(event EventOnStop
  (params (Hit :type Hit%20Result%20Structure%20%28by%20ref%29))
  ;[0.7116 0.838 1] OnStop - Stop the rotation
  (CallFunction|StopRotation))

(event Custom|Splash
  (params (WaterBody :type Fort%20Water%20Body%20Actor%20Object%20Reference) (WaterInteractionComponent :type Fort%20Water%20Interaction%20Component%20Object%20Reference) (bIsFirstBody :type Boolean))
  ;[0.04342 0.3883 1] Grenade landing in water splash
  (Networking|CanExecuteCosmeticEvents
    (:True
      (Utilities|FlowControl|DoOnce false
        (:Completed
          (bind _outwatersurfaceinfo (WaterBody|GetWaterSurfaceInfoStruct WaterBody (Transformation|GetActorLocation) true false false))
          (bind _returnvalue (Effects|Components|ParticleSystem|SpawnFXAtLocation (Variables|VFX|GetNS_HitWater) (Utilities|Struct|BreakWaterSurfaceInfo _outwatersurfaceinfo) (Math|Rotator|MakeRotfromX (.WaterPlaneNormal (Utilities|Struct|BreakWaterSurfaceInfo _outwatersurfaceinfo)))))
          (Effects|Components|ParticleSystem|SetFloatParameter _returnvalue "WaterDepth" (.WaterDepth (Utilities|Struct|BreakWaterSurfaceInfo _outwatersurfaceinfo)))
          (Effects|Components|ParticleSystem|SetFloatParameter _returnvalue "WaterBodyIndex" (Math|Conversions|ToFloat(Integer) (.WaterBodyIdx (Utilities|Struct|BreakWaterSurfaceInfo _outwatersurfaceinfo))))
          (bind _waterfuseaudiocomponent (Variables|Default|GetWaterFuseAudioComponent)
            (:then
              (Audio|Components|Audio|FadeOut _waterfuseaudiocomponent 0.35 0.0 "Linear" :!class /Script/Engine.AudioComponent))
            (:else
              _))
          (bind _returnvalue_2 (Audio|SpawnSoundAttached (Variables|Sounds|GetWaterFuseSound) (Variables|B_Prj_Grenade|GetGrenadeFuse_AudioComponent)))
          (Variables|Default|SetWaterFuseAudioComponent _returnvalue_2))))))

(event Game|Damage|EventAnyDamage
  (params (Damage :type Float%20%28single-precision%29) (DamageType :type Damage%20Type%20Object%20Reference) (InstigatedBy :type Controller%20Object%20Reference) (DamageCauser :type Actor%20Object%20Reference))
  ;[1 1 1] AnyDamage - explode
  (Utilities|FlowControl|Delay 0.1)
  (Delivery|DoExplosion))

(event EventOnExploded
  (params (HitActors :type Array%20of%20Actor%20Object%20References) (HitResults :type Array%20of%20Hit%20Result%20Structures))
  (bind _e_surface_type (Variables|Default|GetE_Surface_Type))
  (bind _returnvalue (Math|Conversions|ToInteger(Byte) _e_surface_type))
  (bind _returnvalue_2 (Transformation|GetActorLocation))
  (bind _cue_explosionsound (Variables|Sounds|GetCue_ExplosionSound))
  ;[1 1 1] OnExploded - Emitter, Sound, Camera Shake
  (Utilities|Time|ClearandInvalidateTimerbyHandle (Variables|Default|GetFuseIndicatorTimer))
  (Components|Movement|SetUpdatedComponent (Variables|Movement|GetProjectileMovementComponent) 0)
  (Collision|SetActorEnableCollision false)
  (Networking|CanExecuteCosmeticEvents
    (:True
      (bind _outhit (Collision|LineTraceByChannel (+ _returnvalue_2 "0.000000,0.000000,50.000000") (+ _returnvalue_2 "0.000000,0.000000,-50.000000") "TraceTypeQuery12"))
      (bind (_outhit_bblockinghit _outhit_binitialoverlap _outhit_time _outhit_distance _outhit_location _outhit_impactpoint _outhit_normal _outhit_impactnormal _outhit_physmat _outhit_hitactor _outhit_hitcomponent) (Collision|BreakHitResult _outhit))
      (Variables|Default|SetExplodedInWaterForAudio (Utilities|String|Contains (Utilities|GetDisplayName _outhit_hitcomponent) "waterbody"))
      (bind _returnvalue_3 (Effects|Components|ParticleSystem|SpawnFXAtLocation (CallFunction|GetExplosion (Utilities|String|Contains (Utilities|GetDisplayName _outhit_hitcomponent) "waterbody")) _returnvalue_2 (Variables|Default|GetExplosion%20Rotation)))
      (Effects|Components|ParticleSystem|SetFloatParameter _returnvalue_3 "Global Scale" (Variables|VFX|GetExplosion%20Water%20Global%20Scale))
      (Effects|Components|ParticleSystem|SetFloatParameter _returnvalue_3 "UseBase" (Variables|VFX|GetUseBaseEmitters))
      ;[1 1 1] PhysSurface Interaction for explosion Niagara FX
      (bind _checkedphysicalsurfaceid (CallFunction|PhysicalSurfaceCheck _returnvalue))
      (if _checkedphysicalsurfaceid
        (bind _returnvalue_4 (Effects|Components|ParticleSystem|SpawnFXAtLocation (Variables|VFX|GetFX_Physical%20Surface) _returnvalue_2))
        (Effects|Components|ParticleSystem|SetIntParameter _returnvalue_4 "PhysicalSurface" _returnvalue)
        (Effects|Components|ParticleSystem|SetFloatParameter _returnvalue_4 "Global Scale" 0.8))
      (Rendering|SetRenderCustomDepth _returnvalue_3 true) ; DISABLED
      (Utilities|FlowControl|Sequence
        (:then_0
          (Camera|PlayWorldCameraShake (Variables|Default|GetExplosionCameraShake) (Transformation|GetActorLocation) 200 800))
        (:then_1
          (Audio|Components|Audio|FadeOut (Variables|B_Prj_Grenade|GetGrenadeFuse_AudioComponent) 0.2 0.0 "Linear" :!class /Script/Engine.AudioComponent) ; Fade Out Fuse Sound
          (Utilities|IsValid _cue_explosionsound
            (:Is%20Valid
              (Audio|PlaySoundatLocation _cue_explosionsound (Transformation|GetActorLocation) "0, 0, 0" 1.0 1.0 0.0 0 0 (Game|GetInstigator)) ; Play Explosion
              (Fortnite|ShowSoundIndicatorLocal :Location (Transformation|GetActorLocation) :MaxAudibleDistance (Variables|/Script/Engine.SoundBase|GetMaxDistance (Variables|Sounds|GetCue_ExplosionSound)) :IndicatorType "Gunshot" :Instigator (Variables|Actor|GetInstigator) :Tint "(R=0.000000,G=0.000000,B=0.000000,A=0.000000)" :SoundBase (Variables|Sounds|GetCue_ExplosionSound) :AttenuationMaxRangeMultiplier 1.0 :AttenuationIntensityMultiplier 1.0) ; Indicator for explosion
            )
            (:Is%20Not%20Valid
              _))
          (bind _waterfuseaudiocomponent (Variables|Default|GetWaterFuseAudioComponent)
            (:then
              (Audio|Components|Audio|FadeOut _waterfuseaudiocomponent 0.2 0.0 "Linear" :!class /Script/Engine.AudioComponent)
              (Variables|Sounds|GetCue_ExplosionSound
                (:then)
                (:else)))
            (:else)))
        (:then_2
          (Fortnite|PlayLocalForceFeedbackatLocationMulti (Variables|Default|GetExplosionForceFeedbackNear) (Variables|Default|GetExplosionForceFeedbackFar) (Transformation|GetActorLocation) 4.5e+02 7e+02 "Explosion"))
        (:then_3
          (Development|SetHiddeninGame (Variables|B_Prj_Grenade|GetMesh) true false)
          (Components|Activation|Deactivate (Variables|Default|GetNS_Fuse_Particle))
          (if (Math|Boolean|ORBoolean (Water|IsinWater (Variables|Water|GetWaterInteractionComponent)) (Utilities|Enum|Equal(Enum) (Variables|Default|GetE_Surface_Type) "SurfaceType8") (Variables|Default|GetExplodedInWaterForAudio))
            (Audio|PlaySoundatLocation (Variables|Sounds|GetWater%20Debris%20Explosion) (Transformation|GetActorLocation))))))))

(event EventOnResumeSimulation
  ;[1 1 1] OnResumeSimulation
  (|CallParentFunction)
  (Transformation|SetActorScale3D "1.000000,1.000000,1.000000")
  (Components|Movement|SetUpdatedComponent (Variables|Default|GetRotatingMovement) (Variables|B_Prj_Grenade|GetMesh)))

(event OnEnterWaterEvent(WaterInteractionComponent)
  (params (WaterBody :type Fort%20Water%20Body%20Actor%20Object%20Reference) (WaterInteractionComponent :type Fort%20Water%20Interaction%20Component%20Object%20Reference) (bIsFirstBody :type Boolean))
  (if bIsFirstBody
    (CallFunction|HeldWaterImpact)))

(event Custom|Held%20Water%20Impact
  (Audio|PlaySoundatLocation (Variables|Sounds|GetSplash%20Sound) (Transformation|GetActorLocation)))

(fn FuseEnded)

(fn GetExplosion
  (params (InWater :type Boolean))
  (return (select InWater (Variables|VFX|GetVFX_Explosion_Water) (Variables|VFX|GetVFX_Explosion))))

(fn PhysicalSurfaceCheck
  (params (Physical%20Surface%20ID :type Integer))
  (return (Math|Boolean|ORBoolean (== Physical%20Surface%20ID 30) (== Physical%20Surface%20ID 26) (== Physical%20Surface%20ID 23) (== Physical%20Surface%20ID 24))))

(fn ConstructionScript
  (Networking|CanExecuteCosmeticEvents
    (:True
      (Components|Movement|SetUpdatedComponent (Variables|Default|GetRotatingMovement) (Variables|B_Prj_Grenade|GetMesh))
      (Audio|SetCollisionSoundatIndex (Variables|Default|GetFortCollisionAudio) 0 (Variables|Sounds|GetCue_Bounce)))))
