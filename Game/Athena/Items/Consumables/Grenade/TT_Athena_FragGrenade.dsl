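; TT_Athena_FragGrenade is a tooltip Blueprint, apparently for the Athena frag grenade consumable.
; It declares three functions that take a gameplay ability instance: application tag lookup, text lookup by token, and instance initialization with a tooltip context.
; All three are stubs here: both lookups return empty or default values with false, and initialization has no body.
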(fn GetApplicationTag
  (params (AbilityInstance :type Gameplay%20Ability%20Object%20Reference))
  (return 0 false))

(fn GetTextForTokenFromAbilityInstance
  (params (AbilityInstance :type Gameplay%20Ability%20Object%20Reference) (Tag :type Gameplay%20Tag%20Structure%20%28by%20ref%29) (Context :type Fort%20Tooltip%20Context%20Object%20Reference) (Token :type Gameplay%20Tag%20Structure%20%28by%20ref%29))
  (return "" false))

(fn InitializeAbilityInstance
  (params (AbilityInstance :type Gameplay%20Ability%20Object%20Reference) (Context :type Fort%20Tooltip%20Context%20Object%20Reference)))
